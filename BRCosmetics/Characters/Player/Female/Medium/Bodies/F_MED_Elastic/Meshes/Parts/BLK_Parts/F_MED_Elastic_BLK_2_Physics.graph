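# Physics asset for the BLK_2 hair part of the female medium Elastic character body, giving hair bones collision bodies and joints for simulation.
# Small capsules on strand bones blk_hair_a through blk_hair_w each constrain directly to the blk2_hair root in a star layout, with tight angular limits.
# The root joins the head with free swing and twist.
# Two large overlapping capsules form the head's collision volume.
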PhysicsAsset /BRCosmetics/Characters/Player/Female/Medium/Bodies/F_MED_Elastic/Meshes/Parts/BLK_Parts/F_MED_Elastic_BLK_2_Physics
bodies {
 blk2_hair [{shapeType=Capsule, length=1, radius=1}]
 blk_hair_a [{shapeType=Capsule, length=2, radius=1.25, center=<1 0 0>, rotation=<rot 90 0 0>}]
 blk_hair_b [{shapeType=Capsule, length=2, radius=1.25, center=<1.3727 0 0>, rotation=<rot 90 0 0>}]
 blk_hair_c [{shapeType=Capsule, length=2, radius=1.25, center=<2 0 0>, rotation=<rot 90 0 0>}]
 blk_hair_d [{shapeType=Capsule, length=2, radius=1.25, center=<1.7242 -0.0001 0>, rotation=<rot 90 0 0>}]
 blk_hair_e [{shapeType=Capsule, length=2, radius=1.25, center=<1.7242 0 0>, rotation=<rot 90 0 0>}]
 blk_hair_f [{shapeType=Capsule, length=2, radius=1.25, center=<2 0 0>, rotation=<rot 90 0 0>}]
 blk_hair_g [{shapeType=Capsule, length=2, radius=1.25, center=<2 0 0>, rotation=<rot 90 0 0>}]
 blk_hair_h [{shapeType=Capsule, length=2, radius=1.25, center=<1 0 0>, rotation=<rot 90 0 0>}]
 blk_hair_i [{shapeType=Capsule, length=2, radius=1.25, center=<0.8113 0 0>, rotation=<rot 90 0 0>}]
 blk_hair_j [{shapeType=Capsule, length=2, radius=1.25, center=<1.5 0 0>, rotation=<rot 90 0 0>}]
 blk_hair_k [{shapeType=Capsule, length=2, radius=1.25, center=<0.9523 0 0>, rotation=<rot 90 0 0>}]
 blk_hair_l [{shapeType=Capsule, length=2, radius=1.25, center=<2 0 0>, rotation=<rot 90 0 0>}]
 blk_hair_m [{shapeType=Capsule, length=2, radius=1.25, center=<2.3499 0 0>, rotation=<rot 90 0 0>}]
 blk_hair_n [{shapeType=Capsule, length=2, radius=1.25, center=<1.5 0 0>, rotation=<rot 90 0 0>}]
 blk_hair_o [{shapeType=Capsule, length=2, radius=1.25, center=<1.7291 0 0>, rotation=<rot 90 0 0>}]
 blk_hair_p [{shapeType=Capsule, length=2, radius=1.25, center=<2 0 0>, rotation=<rot 90 0 0>}]
 blk_hair_q [{shapeType=Capsule, length=2, radius=1.25, center=<1.3678 0 0>, rotation=<rot 90 0 0>}]
 blk_hair_r [{shapeType=Capsule, length=2, radius=1.25, center=<1.3787 -0.0001 0>, rotation=<rot 90 0 0>}]
 blk_hair_s [{shapeType=Capsule, length=1.25, radius=1.25, center=<1.1288 0 0>, rotation=<rot 90 0 0>}]
 blk_hair_t [{shapeType=Capsule, length=2, radius=1.25, center=<2 0 0>, rotation=<rot 90 0 0>}]
 blk_hair_u [{shapeType=Capsule, length=2, radius=1.25, center=<2.2253 0 0>, rotation=<rot 90 0 0>}]
 blk_hair_v [{shapeType=Capsule, length=1.5, radius=1.25, center=<1.4951 0.4696 -0.1805>, rotation=<rot 90 0 0>}]
 blk_hair_w [{shapeType=Capsule, length=1.5, radius=1.25, center=<1.5775 0.0001 0>, rotation=<rot 90 0 0>}]
 head [
  {shapeType=Capsule, length=2.5, radius=7, center=<3.7326 2.2918 0>, rotation=<rot 90 0 0>}
  {shapeType=Capsule, length=0, radius=7.5, center=<6.2855 0.161 0.1937>, rotation=<rot -88.235 0 0>}
 ]
}
constraints [
 {bone1Name=blk_hair_a, bone2Name=blk2_hair, swing1LimitDegrees=30, swing2LimitDegrees=30, twistLimitDegrees=30}
 {bone1Name=blk_hair_b, bone2Name=blk2_hair, swing1LimitDegrees=30, swing2LimitDegrees=30, twistLimitDegrees=30}
 {bone1Name=blk_hair_c, bone2Name=blk2_hair, swing1LimitDegrees=30, swing2LimitDegrees=30, twistLimitDegrees=30}
 {bone1Name=blk_hair_d, bone2Name=blk2_hair, swing1LimitDegrees=30, swing2LimitDegrees=30, twistLimitDegrees=30}
 {bone1Name=blk_hair_e, bone2Name=blk2_hair, swing1LimitDegrees=30, swing2LimitDegrees=30, twistLimitDegrees=30}
 {bone1Name=blk_hair_f, bone2Name=blk2_hair, swing1LimitDegrees=30, swing2LimitDegrees=30, twistLimitDegrees=30}
 {bone1Name=blk_hair_g, bone2Name=blk2_hair, swing1LimitDegrees=30, swing2LimitDegrees=30, twistLimitDegrees=30}
 {bone1Name=blk_hair_h, bone2Name=blk2_hair, swing1LimitDegrees=30, swing2LimitDegrees=30, twistLimitDegrees=30}
 {bone1Name=blk_hair_i, bone2Name=blk2_hair, swing1LimitDegrees=30, swing2LimitDegrees=30, twistLimitDegrees=30}
 {bone1Name=blk_hair_j, bone2Name=blk2_hair, swing1LimitDegrees=30, swing2LimitDegrees=30, twistLimitDegrees=30}
 {bone1Name=blk_hair_k, bone2Name=blk2_hair, swing1LimitDegrees=30, swing2LimitDegrees=30, twistLimitDegrees=30}
 {bone1Name=blk_hair_l, bone2Name=blk2_hair, swing1LimitDegrees=30, swing2LimitDegrees=30, twistLimitDegrees=30}
 {bone1Name=blk_hair_m, bone2Name=blk2_hair, swing1LimitDegrees=30, swing2LimitDegrees=30, twistLimitDegrees=30}
 {bone1Name=blk_hair_n, bone2Name=blk2_hair, swing1LimitDegrees=30, swing2LimitDegrees=30, twistLimitDegrees=30}
 {bone1Name=blk_hair_o, bone2Name=blk2_hair, swing1LimitDegrees=30, swing2LimitDegrees=30, twistLimitDegrees=30}
 {bone1Name=blk_hair_p, bone2Name=blk2_hair, swing1LimitDegrees=30, swing2LimitDegrees=30, twistLimitDegrees=30}
 {bone1Name=blk_hair_q, bone2Name=blk2_hair, swing1LimitDegrees=30, swing2LimitDegrees=30, twistLimitDegrees=30}
 {bone1Name=blk_hair_r, bone2Name=blk2_hair, swing1LimitDegrees=30, swing2LimitDegrees=30, twistLimitDegrees=30}
 {bone1Name=blk_hair_s, bone2Name=blk2_hair, swing1LimitDegrees=25, swing2LimitDegrees=25, twistLimitDegrees=25}
 {bone1Name=blk_hair_t, bone2Name=blk2_hair, swing1LimitDegrees=30, swing2LimitDegrees=30, twistLimitDegrees=30}
 {bone1Name=blk_hair_u, bone2Name=blk2_hair, swing1LimitDegrees=30, swing2LimitDegrees=30, twistLimitDegrees=30}
 {
  bone1Name=blk2_hair
  bone2Name=head
  swing1LimitDegrees=45
  swing1Motion=Free
  swing2LimitDegrees=45
  swing2Motion=Free
  twistLimitDegrees=45
  twistMotion=Free
 }
 {bone1Name=blk_hair_v, bone2Name=blk2_hair, swing1LimitDegrees=45, swing2LimitDegrees=45, twistLimitDegrees=45}
 {bone1Name=blk_hair_w, bone2Name=blk2_hair, swing1LimitDegrees=45, swing2LimitDegrees=45, twistLimitDegrees=45}
]
